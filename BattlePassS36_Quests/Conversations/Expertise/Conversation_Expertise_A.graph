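# Conversation database for the Season 36 Expertise A NPC (IronLilac lines), covering Unstable Shotgun expertise quest dialogue and discounted shotgun sales.
# One entry point shows intro or completion dialogue while the matching quest is active.
# The other offers Unstable and Epic shotgun Expert Sales, each duplicated across five token levels, with failing purchases hidden and cost rows chosen by token count.
# Most branches return to start; Unstable purchases at levels 4 and 5 have no output.
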ConversationDatabase /BattlePassS36_Quests/Conversations/Expertise/Conversation_Expertise_A
entry_points [{tag=Conversation.Entry.Expertise.A, destinations [#1, #2]}]
nodes [
 ConversationEntryPointNode_0 {
  id=#2
  class=ConversationEntryPointNode
  outputs [
   #3
   #4
   #5
   #6
   #7
   #8
   #9
   #10
   #11
   #12
  ]
  properties {entryTag {tagName=Conversation.Entry.Expertise.A}}
 }
 Speech_C_14 {
  id=#12
  class=Speech_C
  outputs [#13, #14]
  properties {
   bUsePerNPC_ContextualMessageLookup=true
   softTaskIcon=&T_UI_NPC_WeaponExpertise
   generalConfig {defaultMessage=You want this?}
   speakerEntryTagToConfig {"(TagName=\"Conversation.Entry.IronLilac\")" {defaultMessage=Make an impact.}}
  }
  sub_nodes [
   FortConversationChoiceWithPreviewNode_0 {class=FortConversationChoiceWithPreviewNode, properties {defaultChoiceDisplayText=EXPERT SALE}}
   FortConversationRequirement_HasService_0 {class=FortConversationRequirement_HasService, properties {serviceTag {tagName=Tandem.Service.ExpertSale}}}
  ]
 }
 Back_C_10 {
  id=#14
  class=Back_C
  sub_nodes [ConversationChoiceNode_22 {class=ConversationChoiceNode, properties {defaultChoiceDisplayText=BACK}}]
 }
 ExpertSale_C_42 {
  id=#13
  class=ExpertSale_C
  outputs [#15]
  properties {
   bShowPropertyDetails=false
   loot Tier Group Name=LTG_NPC_Expertise_Shotgun_VR
   serviceBriefCollectionKeyOveride=None
   serviceBriefConfigCollection=null
   requirements [{
     cannotUseReasonParameter=""
     failureNodeBehaviour=FailedAndHidden
     participantId {tagName=Conversation.Participant.Player}
     requirement {
      class=UFortControllerRequirement_TokenThreshold_C
      properties {
       itemTemplateId=Token:athena_s36_weaponexpertise_unstableshotgun_service_token
       requiredQuantityThresholdConfig {requirement=Equal, quantity {value=5}}
      }
     }
    }]
   serviceCost {curve {rowName=ExpertSale.Cost.EpicShotgun.5}}
  }
  sub_nodes [ConversationChoiceNode_35 {class=ConversationChoiceNode, properties {defaultChoiceDisplayText=BUY}}]
 }
 ReturnToStart_C_5 {id=#15, class=ReturnToStart_C}
 Speech_C_13 {
  id=#11
  class=Speech_C
  outputs [#16, #17]
  properties {
   bUsePerNPC_ContextualMessageLookup=true
   softTaskIcon=&T_UI_NPC_WeaponExpertise
   generalConfig {defaultMessage=You want this?}
   speakerEntryTagToConfig {"(TagName=\"Conversation.Entry.IronLilac\")" {defaultMessage=Make an impact.}}
  }
  sub_nodes [
   FortConversationChoiceWithPreviewNode_10 {class=FortConversationChoiceWithPreviewNode, properties {defaultChoiceDisplayText=EXPERT SALE}}
   FortConversationRequirement_HasService_10 {class=FortConversationRequirement_HasService, properties {serviceTag {tagName=Tandem.Service.ExpertSale}}}
  ]
 }
 Back_C_8 {
  id=#17
  class=Back_C
  sub_nodes [ConversationChoiceNode_44 {class=ConversationChoiceNode, properties {defaultChoiceDisplayText=BACK}}]
 }
 ExpertSale_C_41 {
  id=#16
  class=ExpertSale_C
  outputs [#15]
  properties {
   bShowPropertyDetails=false
   loot Tier Group Name=LTG_NPC_Expertise_Shotgun_VR
   serviceBriefCollectionKeyOveride=None
   serviceBriefConfigCollection=null
   requirements [{
     cannotUseReasonParameter=""
     failureNodeBehaviour=FailedAndHidden
     participantId {tagName=Conversation.Participant.Player}
     requirement {
      class=UFortControllerRequirement_TokenThreshold_C
      properties {
       itemTemplateId=Token:athena_s36_weaponexpertise_unstableshotgun_service_token
       requiredQuantityThresholdConfig {requirement=Equal, quantity {value=4}}
      }
     }
    }]
   serviceCost {curve {rowName=ExpertSale.Cost.EpicShotgun.4}}
  }
  sub_nodes [ConversationChoiceNode_43 {class=ConversationChoiceNode, properties {defaultChoiceDisplayText=BUY}}]
 }
 Speech_C_10 {
  id=#10
  class=Speech_C
  outputs [#18, #19]
  properties {
   bUsePerNPC_ContextualMessageLookup=true
   softTaskIcon=&T_UI_NPC_WeaponExpertise
   generalConfig {defaultMessage=You want this?}
   speakerEntryTagToConfig {"(TagName=\"Conversation.Entry.IronLilac\")" {defaultMessage=Make an impact.}}
  }
  sub_nodes [
   FortConversationChoiceWithPreviewNode_9 {class=FortConversationChoiceWithPreviewNode, properties {defaultChoiceDisplayText=EXPERT SALE}}
   FortConversationRequirement_HasService_9 {class=FortConversationRequirement_HasService, properties {serviceTag {tagName=Tandem.Service.ExpertSale}}}
  ]
 }
 Back_C_7 {
  id=#19
  class=Back_C
  sub_nodes [ConversationChoiceNode_42 {class=ConversationChoiceNode, properties {defaultChoiceDisplayText=BACK}}]
 }
 ExpertSale_C_40 {
  id=#18
  class=ExpertSale_C
  outputs [#15]
  properties {
   bShowPropertyDetails=false
   loot Tier Group Name=LTG_NPC_Expertise_Shotgun_VR
   serviceBriefCollectionKeyOveride=None
   serviceBriefConfigCollection=null
   requirements [{
     cannotUseReasonParameter=""
     failureNodeBehaviour=FailedAndHidden
     participantId {tagName=Conversation.Participant.Player}
     requirement {
      class=UFortControllerRequirement_TokenThreshold_C
      properties {
       itemTemplateId=Token:athena_s36_weaponexpertise_unstableshotgun_service_token
       requiredQuantityThresholdConfig {requirement=Equal, quantity {value=3}}
      }
     }
    }]
   serviceCost {curve {rowName=ExpertSale.Cost.EpicShotgun.3}}
  }
  sub_nodes [ConversationChoiceNode_41 {class=ConversationChoiceNode, properties {defaultChoiceDisplayText=BUY}}]
 }
 Speech_C_9 {
  id=#9
  class=Speech_C
  outputs [#20, #21]
  properties {
   bUsePerNPC_ContextualMessageLookup=true
   softTaskIcon=&T_UI_NPC_WeaponExpertise
   generalConfig {defaultMessage=You want this?}
   speakerEntryTagToConfig {"(TagName=\"Conversation.Entry.IronLilac\")" {defaultMessage=Make an impact.}}
  }
  sub_nodes [
   FortConversationChoiceWithPreviewNode_7 {class=FortConversationChoiceWithPreviewNode, properties {defaultChoiceDisplayText=EXPERT SALE}}
   FortConversationRequirement_HasService_7 {class=FortConversationRequirement_HasService, properties {serviceTag {tagName=Tandem.Service.ExpertSale}}}
  ]
 }
 Back_C_6 {
  id=#21
  class=Back_C
  sub_nodes [ConversationChoiceNode_40 {class=ConversationChoiceNode, properties {defaultChoiceDisplayText=BACK}}]
 }
 ExpertSale_C_39 {
  id=#20
  class=ExpertSale_C
  outputs [#15]
  properties {
   bShowPropertyDetails=false
   loot Tier Group Name=LTG_NPC_Expertise_Shotgun_VR
   serviceBriefCollectionKeyOveride=None
   serviceBriefConfigCollection=null
   requirements [{
     cannotUseReasonParameter=""
     failureNodeBehaviour=FailedAndHidden
     participantId {tagName=Conversation.Participant.Player}
     requirement {
      class=UFortControllerRequirement_TokenThreshold_C
      properties {
       itemTemplateId=Token:athena_s36_weaponexpertise_unstableshotgun_service_token
       requiredQuantityThresholdConfig {requirement=Equal, quantity {value=2}}
      }
     }
    }]
   serviceCost {curve {rowName=ExpertSale.Cost.EpicShotgun.2}}
  }
  sub_nodes [ConversationChoiceNode_39 {class=ConversationChoiceNode, properties {defaultChoiceDisplayText=BUY}}]
 }
 Speech_C_8 {
  id=#8
  class=Speech_C
  outputs [#22, #23]
  properties {
   bUsePerNPC_ContextualMessageLookup=true
   softTaskIcon=&T_UI_NPC_WeaponExpertise
   generalConfig {defaultMessage=You want this?}
   speakerEntryTagToConfig {"(TagName=\"Conversation.Entry.IronLilac\")" {defaultMessage=Make an impact.}}
  }
  sub_nodes [
   FortConversationChoiceWithPreviewNode_6 {class=FortConversationChoiceWithPreviewNode, properties {defaultChoiceDisplayText=EXPERT SALE}}
   FortConversationRequirement_HasService_6 {class=FortConversationRequirement_HasService, properties {serviceTag {tagName=Tandem.Service.ExpertSale}}}
  ]
 }
 Back_C_5 {
  id=#23
  class=Back_C
  sub_nodes [ConversationChoiceNode_38 {class=ConversationChoiceNode, properties {defaultChoiceDisplayText=BACK}}]
 }
 ExpertSale_C_38 {
  id=#22
  class=ExpertSale_C
  outputs [#15]
  properties {
   bShowPropertyDetails=false
   loot Tier Group Name=LTG_NPC_Expertise_Shotgun_VR
   serviceBriefCollectionKeyOveride=None
   serviceBriefConfigCollection=null
   requirements [{
     cannotUseReasonParameter=""
     failureNodeBehaviour=FailedAndHidden
     participantId {tagName=Conversation.Participant.Player}
     requirement {
      class=UFortControllerRequirement_TokenThreshold_C
      properties {
       itemTemplateId=Token:athena_s36_weaponexpertise_unstableshotgun_service_token
       requiredQuantityThresholdConfig {requirement=LessThanOrEqual, quantity {value=1}}
      }
     }
    }]
   serviceCost {curve {rowName=ExpertSale.Cost.EpicShotgun.1}}
  }
  sub_nodes [ConversationChoiceNode_37 {class=ConversationChoiceNode, properties {defaultChoiceDisplayText=BUY}}]
 }
 Speech_C_6 {
  id=#7
  class=Speech_C
  outputs [#24, #25]
  properties {
   bUsePerNPC_ContextualMessageLookup=true
   softTaskIcon=&T_UI_NPC_WeaponExpertise
   generalConfig {defaultMessage=You want this?}
   speakerEntryTagToConfig {"(TagName=\"Conversation.Entry.IronLilac\")" {defaultMessage=You've earned this}}
  }
  sub_nodes [
   FortConversationChoiceWithPreviewNode_2 {class=FortConversationChoiceWithPreviewNode, properties {defaultChoiceDisplayText=EXPERT SALE}}
   FortConversationRequirement_HasService_2 {class=FortConversationRequirement_HasService, properties {serviceTag {tagName=Tandem.Service.ExpertSale}}}
  ]
 }
 Back_C_3 {
  id=#25
  class=Back_C
  sub_nodes [ConversationChoiceNode_30 {class=ConversationChoiceNode, properties {defaultChoiceDisplayText=BACK}}]
 }
 ExpertSale_C_36 {
  id=#24
  class=ExpertSale_C
  outputs []
  properties {
   bShowPropertyDetails=false
   loot Tier Group Name=LTG_NPC_Expertise_Shotgun_MR
   serviceBriefCollectionKeyOveride=None
   serviceBriefConfigCollection=null
   requirements [{
     cannotUseReasonParameter=""
     failureNodeBehaviour=FailedAndHidden
     participantId {tagName=Conversation.Participant.Player}
     requirement {
      class=UFortControllerRequirement_TokenThreshold_C
      properties {
       itemTemplateId=Token:athena_s36_weaponexpertise_unstableshotgun_service_token
       requiredQuantityThresholdConfig {requirement=Equal, quantity {value=4}}
      }
     }
    }]
   serviceCost {curve {rowName=ExpertSale.Cost.UnstableShotgun.4}}
  }
  sub_nodes [ConversationChoiceNode_29 {class=ConversationChoiceNode, properties {defaultChoiceDisplayText=BUY}}]
 }
 Speech_C_7 {
  id=#6
  class=Speech_C
  outputs [#26, #27]
  properties {
   bUsePerNPC_ContextualMessageLookup=true
   softTaskIcon=&T_UI_NPC_WeaponExpertise
   generalConfig {defaultMessage=You want this?}
   speakerEntryTagToConfig {"(TagName=\"Conversation.Entry.IronLilac\")" {defaultMessage=You've earned this}}
  }
  sub_nodes [
   FortConversationChoiceWithPreviewNode_8 {class=FortConversationChoiceWithPreviewNode, properties {defaultChoiceDisplayText=EXPERT SALE}}
   FortConversationRequirement_HasService_8 {class=FortConversationRequirement_HasService, properties {serviceTag {tagName=Tandem.Service.ExpertSale}}}
  ]
 }
 Back_C_4 {
  id=#27
  class=Back_C
  sub_nodes [ConversationChoiceNode_23 {class=ConversationChoiceNode, properties {defaultChoiceDisplayText=BACK}}]
 }
 ExpertSale_C_37 {
  id=#26
  class=ExpertSale_C
  outputs []
  properties {
   bShowPropertyDetails=false
   loot Tier Group Name=LTG_NPC_Expertise_Shotgun_MR
   serviceBriefCollectionKeyOveride=None
   serviceBriefConfigCollection=null
   requirements [{
     cannotUseReasonParameter=""
     failureNodeBehaviour=FailedAndHidden
     participantId {tagName=Conversation.Participant.Player}
     requirement {
      class=UFortControllerRequirement_TokenThreshold_C
      properties {
       itemTemplateId=Token:athena_s36_weaponexpertise_unstableshotgun_service_token
       requiredQuantityThresholdConfig {requirement=Equal, quantity {value=5}}
      }
     }
    }]
   serviceCost {curve {rowName=ExpertSale.Cost.UnstableShotgun.5}}
  }
  sub_nodes [ConversationChoiceNode_24 {class=ConversationChoiceNode, properties {defaultChoiceDisplayText=BUY}}]
 }
 Speech_C_5 {
  id=#5
  class=Speech_C
  outputs [#28, #29]
  properties {
   bUsePerNPC_ContextualMessageLookup=true
   softTaskIcon=&T_UI_NPC_WeaponExpertise
   generalConfig {defaultMessage=You want this?}
   speakerEntryTagToConfig {"(TagName=\"Conversation.Entry.IronLilac\")" {defaultMessage=You've earned this}}
  }
  sub_nodes [
   FortConversationChoiceWithPreviewNode_1 {class=FortConversationChoiceWithPreviewNode, properties {defaultChoiceDisplayText=EXPERT SALE}}
   FortConversationRequirement_HasService_1 {class=FortConversationRequirement_HasService, properties {serviceTag {tagName=Tandem.Service.ExpertSale}}}
  ]
 }
 Back_C_0 {
  id=#29
  class=Back_C
  sub_nodes [ConversationChoiceNode_27 {class=ConversationChoiceNode, properties {defaultChoiceDisplayText=BACK}}]
 }
 ExpertSale_C_33 {
  id=#28
  class=ExpertSale_C
  outputs [#30]
  properties {
   bShowPropertyDetails=false
   loot Tier Group Name=LTG_NPC_Expertise_Shotgun_MR
   serviceBriefCollectionKeyOveride=None
   serviceBriefConfigCollection=null
   requirements [{
     cannotUseReasonParameter=""
     failureNodeBehaviour=FailedAndHidden
     participantId {tagName=Conversation.Participant.Player}
     requirement {
      class=UFortControllerRequirement_TokenThreshold_C
      properties {
       itemTemplateId=Token:athena_s36_weaponexpertise_unstableshotgun_service_token
       requiredQuantityThresholdConfig {requirement=LessThanOrEqual, quantity {value=1}}
      }
     }
    }]
   serviceCost {curve {rowName=ExpertSale.Cost.UnstableShotgun.1}}
  }
  sub_nodes [ConversationChoiceNode_26 {class=ConversationChoiceNode, properties {defaultChoiceDisplayText=BUY}}]
 }
 ReturnToStart_C_4 {id=#30, class=ReturnToStart_C}
 Speech_C_0 {
  id=#4
  class=Speech_C
  outputs [#31, #32]
  properties {
   bUsePerNPC_ContextualMessageLookup=true
   softTaskIcon=&T_UI_NPC_WeaponExpertise
   generalConfig {defaultMessage=You want this?}
   speakerEntryTagToConfig {"(TagName=\"Conversation.Entry.IronLilac\")" {defaultMessage=You've earned this}}
  }
  sub_nodes [
   FortConversationChoiceWithPreviewNode_3 {class=FortConversationChoiceWithPreviewNode, properties {defaultChoiceDisplayText=EXPERT SALE}}
   FortConversationRequirement_HasService_3 {class=FortConversationRequirement_HasService, properties {serviceTag {tagName=Tandem.Service.ExpertSale}}}
  ]
 }
 Back_C_1 {
  id=#32
  class=Back_C
  sub_nodes [ConversationChoiceNode_32 {class=ConversationChoiceNode, properties {defaultChoiceDisplayText=BACK}}]
 }
 ExpertSale_C_34 {
  id=#31
  class=ExpertSale_C
  outputs [#30]
  properties {
   bShowPropertyDetails=false
   loot Tier Group Name=LTG_NPC_Expertise_Shotgun_MR
   serviceBriefCollectionKeyOveride=None
   serviceBriefConfigCollection=null
   requirements [{
     cannotUseReasonParameter=""
     failureNodeBehaviour=FailedAndHidden
     participantId {tagName=Conversation.Participant.Player}
     requirement {
      class=UFortControllerRequirement_TokenThreshold_C
      properties {
       itemTemplateId=Token:athena_s36_weaponexpertise_unstableshotgun_service_token
       requiredQuantityThresholdConfig {requirement=Equal, quantity {value=2}}
      }
     }
    }]
   serviceCost {curve {rowName=ExpertSale.Cost.UnstableShotgun.2}}
  }
  sub_nodes [ConversationChoiceNode_31 {class=ConversationChoiceNode, properties {defaultChoiceDisplayText=BUY}}]
 }
 Speech_C_1 {
  id=#3
  class=Speech_C
  outputs [#33, #34]
  properties {
   bUsePerNPC_ContextualMessageLookup=true
   softTaskIcon=&T_UI_NPC_WeaponExpertise
   generalConfig {defaultMessage=You want this?}
   speakerEntryTagToConfig {"(TagName=\"Conversation.Entry.IronLilac\")" {defaultMessage=You've earned this}}
  }
  sub_nodes [
   FortConversationChoiceWithPreviewNode_4 {class=FortConversationChoiceWithPreviewNode, properties {defaultChoiceDisplayText=EXPERT SALE}}
   FortConversationRequirement_HasService_4 {class=FortConversationRequirement_HasService, properties {serviceTag {tagName=Tandem.Service.ExpertSale}}}
  ]
 }
 Back_C_2 {
  id=#34
  class=Back_C
  sub_nodes [ConversationChoiceNode_34 {class=ConversationChoiceNode, properties {defaultChoiceDisplayText=BACK}}]
 }
 ExpertSale_C_35 {
  id=#33
  class=ExpertSale_C
  outputs [#30]
  properties {
   bShowPropertyDetails=false
   loot Tier Group Name=LTG_NPC_Expertise_Shotgun_MR
   serviceBriefCollectionKeyOveride=None
   serviceBriefConfigCollection=null
   requirements [{
     cannotUseReasonParameter=""
     failureNodeBehaviour=FailedAndHidden
     participantId {tagName=Conversation.Participant.Player}
     requirement {
      class=UFortControllerRequirement_TokenThreshold_C
      properties {
       itemTemplateId=Token:athena_s36_weaponexpertise_unstableshotgun_service_token
       requiredQuantityThresholdConfig {requirement=Equal, quantity {value=3}}
      }
     }
    }]
   serviceCost {curve {rowName=ExpertSale.Cost.UnstableShotgun.3}}
  }
  sub_nodes [ConversationChoiceNode_33 {class=ConversationChoiceNode, properties {defaultChoiceDisplayText=BUY}}]
 }
 ConversationEntryPointNode_2 {
  id=#1
  class=ConversationEntryPointNode
  outputs [#35, #36]
  properties {entryTag {tagName=Conversation.Entry.Expertise.A}}
 }
 Speech_C_4 {
  id=#36
  class=Speech_C
  outputs [#37]
  properties {
   softTaskIcon=&T_UI_NPC_MainQuest
   generalConfig {defaultMessage="Looks like you're top of the class! Well done, expert.", defaultMetadataParameters [IronLilac {value=""}]}
  }
  sub_nodes [
   ConversationChoiceNode_28 {
    class=ConversationChoiceNode
    properties {
     defaultChoiceDisplayText=Show off your Expertise
     choiceTags {gameplayTags [{tagName=Athena.Quests.S36.Conversations.IronLilac.Finished.01}]}
    }
   }
   FortConversationRequirement_ControllerRequirement_28 {
    class=FortConversationRequirement_ControllerRequirement
    properties {
     nodeName=Require: Is Quest Active
     requirement {
      class=FortControllerRequirement_IsQuestActive
      properties {questDefinition=/BattlePassS36_Quests/Items/QuestItems/WeaponExpertise/UnstableShotgun/quest_S36_expertise_UnstableShotgun_finished}
     }
    }
   }
  ]
 }
 Speech_C_2 {
  id=#37
  class=Speech_C
  outputs [#38]
  properties {generalConfig {defaultMessage=Please enjoy this honorary certificate from Supernova Academy... and a 60% discount.}}
  sub_nodes [ConversationChoiceNode_16 {
    class=ConversationChoiceNode
    properties {defaultChoiceDisplayText=Continue, choiceTags {gameplayTags [{tagName=Athena.Quests.S36.Conversations.IronLilac.Finished.02}]}}
   }]
 }
 ReturnToStart_C_2 {id=#38, class=ReturnToStart_C}
 Speech_C_11 {
  id=#35
  class=Speech_C
  outputs [#39]
  properties {
   softTaskIcon=&T_UI_NPC_MainQuest
   generalConfig {
    defaultMessage=You know your way around a shotgun. We're looking for experts to test some prototypes.
    defaultMetadataParameters [IronLilac {value=""}]
   }
  }
  sub_nodes [
   ConversationChoiceNode_13 {
    class=ConversationChoiceNode
    properties {
     defaultChoiceDisplayText=Ask about Unstable Exotic Expertise
     choiceTags {gameplayTags [{tagName=Athena.Quests.S36.Conversations.IronLilac.Intro.01}]}
    }
   }
   FortConversationRequirement_ControllerRequirement_10 {
    class=FortConversationRequirement_ControllerRequirement
    properties {
     nodeName=Require: Is Quest Active
     requirement {
      class=FortControllerRequirement_IsQuestActive
      properties {questDefinition=/BattlePassS36_Quests/Items/QuestItems/WeaponExpertise/UnstableShotgun/quest_S36_expertise_UnstableShotgun_intro}
     }
    }
   }
  ]
 }
 Speech_C_12 {
  id=#39
  class=Speech_C
  outputs [#40]
  properties {generalConfig {defaultMessage=Happy to share an experimental shotgun with you... for a small donation to the Academy.}}
  sub_nodes [ConversationChoiceNode_14 {
    class=ConversationChoiceNode
    properties {defaultChoiceDisplayText=Continue, choiceTags {gameplayTags [{tagName=Athena.Quests.S36.Conversations.IronLilac.Intro.02}]}}
   }]
 }
 ReturnToStart_C_1 {id=#40, class=ReturnToStart_C}
]
