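# NPC conversation database for Season 36 Unstable Pistol weapon expertise, selling discounted pistols under the Conversation.Entry.Expertise.B tag.
# Three entry nodes share that tag: intro and finished dialogue, each gated by its active quest, plus ten Expert Sale branches gated by NPC service.
# Sales cover Exotic and Epic pistols in five tiers each.
# Expertise token count selects each tier's cost-curve row and hides non-matching purchases, with lowest tiers carrying locked-display data.
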ConversationDatabase /BattlePassS36_Quests/Conversations/Expertise/Conversation_Expertise_B
entry_points [{
  tag=Conversation.Entry.Expertise.B
  destinations [#1, #2, #3]
 }]
nodes [
 ConversationEntryPointNode_1 {
  id=#3
  class=ConversationEntryPointNode
  outputs [#4]
  properties {entryTag {tagName=Conversation.Entry.Expertise.B}}
 }
 Speech_C_4 {
  id=#4
  class=Speech_C
  outputs [#5]
  properties {
   softTaskIcon=&T_UI_NPC_MainQuest
   generalConfig {defaultMessage=Huh. Not bad. Congrats., defaultMetadataParameters [AuraCop {value=""}]}
  }
  sub_nodes [
   ConversationChoiceNode_1 {
    class=ConversationChoiceNode
    properties {
     defaultChoiceDisplayText=Show off your Expertise
     choiceTags {gameplayTags [{tagName=Athena.Quests.S36.Conversations.UndergroundRebel.Finished.01}]}
    }
   }
   FortConversationRequirement_ControllerRequirement_1 {
    class=FortConversationRequirement_ControllerRequirement
    properties {
     nodeName=Require: Is Quest Active
     requirement {
      class=FortControllerRequirement_IsQuestActive
      properties {questDefinition=/BattlePassS36_Quests/Items/QuestItems/WeaponExpertise/UnstablePistol/quest_S36_expertise_UnstablePistol_finished}
     }
    }
   }
  ]
 }
 Speech_C_2 {
  id=#5
  class=Speech_C
  outputs [#6]
  properties {generalConfig {defaultMessage=Looks like I can give you a 60% discount on our experimental line of pistols.}}
  sub_nodes [ConversationChoiceNode_16 {
    class=ConversationChoiceNode
    properties {defaultChoiceDisplayText=Continue, choiceTags {gameplayTags [{tagName=Athena.Quests.S36.Conversations.UndergroundRebel.Finished.02}]}}
   }]
 }
 ReturnToStart_C_1 {id=#6, class=ReturnToStart_C}
 ConversationEntryPointNode_0 {
  id=#2
  class=ConversationEntryPointNode
  outputs [#7]
  properties {entryTag {tagName=Conversation.Entry.Expertise.B}}
 }
 Speech_C_11 {
  id=#7
  class=Speech_C
  outputs [#8]
  properties {
   softTaskIcon=&T_UI_NPC_MainQuest
   generalConfig {defaultMessage=Hey. You wanna test some experimental pistols?, defaultMetadataParameters [AuraCop {value=""}]}
  }
  sub_nodes [
   ConversationChoiceNode_13 {
    class=ConversationChoiceNode
    properties {
     defaultChoiceDisplayText=Ask about Unstable Exotic Expertise
     choiceTags {gameplayTags [{tagName=Athena.Quests.S36.Conversations.UndergroundRebel.Intro.01}]}
    }
   }
   FortConversationRequirement_ControllerRequirement_10 {
    class=FortConversationRequirement_ControllerRequirement
    properties {
     nodeName=Require: Is Quest Active
     requirement {
      class=FortControllerRequirement_IsQuestActive
      properties {questDefinition=/BattlePassS36_Quests/Items/QuestItems/WeaponExpertise/UnstablePistol/quest_S36_expertise_UnstablePistol_intro}
     }
    }
   }
  ]
 }
 Speech_C_12 {
  id=#8
  class=Speech_C
  outputs [#9]
  properties {generalConfig {defaultMessage="I've got something a little weirder for sale, if you want to give it a shot."}}
  sub_nodes [ConversationChoiceNode_14 {
    class=ConversationChoiceNode
    properties {defaultChoiceDisplayText=Continue, choiceTags {gameplayTags [{tagName=Athena.Quests.S36.Conversations.UndergroundRebel.Intro.02}]}}
   }]
 }
 ReturnToStart_C_0 {id=#9, class=ReturnToStart_C}
 ConversationEntryPointNode_4 {
  id=#1
  class=ConversationEntryPointNode
  outputs [
   #10
   #11
   #12
   #13
   #14
   #15
   #16
   #17
   #18
   #19
  ]
  properties {entryTag {tagName=Conversation.Entry.Expertise.B}}
 }
 Speech_C_13 {
  id=#19
  class=Speech_C
  outputs [#20, #21]
  properties {
   bUsePerNPC_ContextualMessageLookup=true
   softTaskIcon=&T_UI_NPC_WeaponExpertise
   generalConfig {defaultMessage=You've earned this.}
  }
  sub_nodes [
   FortConversationChoiceWithPreviewNode_0 {class=FortConversationChoiceWithPreviewNode, properties {defaultChoiceDisplayText=EXPERT SALE}}
   FortConversationRequirement_HasService_0 {class=FortConversationRequirement_HasService, properties {serviceTag {tagName=Tandem.Service.ExpertSale}}}
  ]
 }
 Back_C_7 {
  id=#21
  class=Back_C
  sub_nodes [ConversationChoiceNode_37 {class=ConversationChoiceNode, properties {defaultChoiceDisplayText=BACK}}]
 }
 ExpertSale_C_24 {
  id=#20
  class=ExpertSale_C
  outputs [#22]
  properties {
   bShowPropertyDetails=false
   loot Tier Group Name=LTG_NPC_Expertise_Pistol_VR
   serviceBriefCollectionKeyOveride=None
   serviceBriefConfigCollection=null
   requirements [{
     cannotUseReasonParameter=""
     failureNodeBehaviour=FailedAndHidden
     participantId {tagName=Conversation.Participant.Player}
     requirement {
      class=UFortControllerRequirement_TokenThreshold_C
      properties {
       itemTemplateId=Token:athena_s36_weaponexpertise_unstablepistol_service_token
       requiredQuantityThresholdConfig {requirement=Equal, quantity {value=5}}
      }
     }
    }]
   serviceCost {curve {rowName=ExpertSale.Cost.EpicPistol.5}}
  }
  sub_nodes [ConversationChoiceNode_31 {class=ConversationChoiceNode, properties {defaultChoiceDisplayText=BUY}}]
 }
 ReturnToStart_C_3 {id=#22, class=ReturnToStart_C}
 Speech_C_10 {
  id=#18
  class=Speech_C
  outputs [#23, #24]
  properties {
   bUsePerNPC_ContextualMessageLookup=true
   softTaskIcon=&T_UI_NPC_WeaponExpertise
   generalConfig {defaultMessage=You've earned this.}
  }
  sub_nodes [
   FortConversationChoiceWithPreviewNode_9 {class=FortConversationChoiceWithPreviewNode, properties {defaultChoiceDisplayText=EXPERT SALE}}
   FortConversationRequirement_HasService_9 {class=FortConversationRequirement_HasService, properties {serviceTag {tagName=Tandem.Service.ExpertSale}}}
  ]
 }
 Back_C_8 {
  id=#24
  class=Back_C
  sub_nodes [ConversationChoiceNode_35 {class=ConversationChoiceNode, properties {defaultChoiceDisplayText=BACK}}]
 }
 ExpertSale_C_23 {
  id=#23
  class=ExpertSale_C
  outputs [#22]
  properties {
   bShowPropertyDetails=false
   loot Tier Group Name=LTG_NPC_Expertise_Pistol_VR
   serviceBriefCollectionKeyOveride=None
   serviceBriefConfigCollection=null
   requirements [{
     cannotUseReasonParameter=""
     failureNodeBehaviour=FailedAndHidden
     participantId {tagName=Conversation.Participant.Player}
     requirement {
      class=UFortControllerRequirement_TokenThreshold_C
      properties {
       itemTemplateId=Token:athena_s36_weaponexpertise_unstablepistol_service_token
       requiredQuantityThresholdConfig {requirement=Equal, quantity {value=4}}
      }
     }
    }]
   serviceCost {curve {rowName=ExpertSale.Cost.EpicPistol.4}}
  }
  sub_nodes [ConversationChoiceNode_34 {class=ConversationChoiceNode, properties {defaultChoiceDisplayText=BUY}}]
 }
 Speech_C_9 {
  id=#17
  class=Speech_C
  outputs [#25, #26]
  properties {
   bUsePerNPC_ContextualMessageLookup=true
   softTaskIcon=&T_UI_NPC_WeaponExpertise
   generalConfig {defaultMessage=You've earned this.}
  }
  sub_nodes [
   FortConversationChoiceWithPreviewNode_8 {class=FortConversationChoiceWithPreviewNode, properties {defaultChoiceDisplayText=EXPERT SALE}}
   FortConversationRequirement_HasService_8 {class=FortConversationRequirement_HasService, properties {serviceTag {tagName=Tandem.Service.ExpertSale}}}
  ]
 }
 Back_C_10 {
  id=#26
  class=Back_C
  sub_nodes [ConversationChoiceNode_10 {class=ConversationChoiceNode, properties {defaultChoiceDisplayText=BACK}}]
 }
 ExpertSale_C_22 {
  id=#25
  class=ExpertSale_C
  outputs [#22]
  properties {
   bShowPropertyDetails=false
   loot Tier Group Name=LTG_NPC_Expertise_Pistol_VR
   serviceBriefCollectionKeyOveride=None
   serviceBriefConfigCollection=null
   requirements [{
     cannotUseReasonParameter=""
     failureNodeBehaviour=FailedAndHidden
     participantId {tagName=Conversation.Participant.Player}
     requirement {
      class=UFortControllerRequirement_TokenThreshold_C
      properties {
       itemTemplateId=Token:athena_s36_weaponexpertise_unstablepistol_service_token
       requiredQuantityThresholdConfig {requirement=Equal, quantity {value=3}}
      }
     }
    }]
   serviceCost {curve {rowName=ExpertSale.Cost.EpicPistol.3}}
  }
  sub_nodes [ConversationChoiceNode_33 {class=ConversationChoiceNode, properties {defaultChoiceDisplayText=BUY}}]
 }
 Speech_C_8 {
  id=#16
  class=Speech_C
  outputs [#27, #28]
  properties {
   bUsePerNPC_ContextualMessageLookup=true
   softTaskIcon=&T_UI_NPC_WeaponExpertise
   generalConfig {defaultMessage=You've earned this.}
  }
  sub_nodes [
   FortConversationChoiceWithPreviewNode_7 {class=FortConversationChoiceWithPreviewNode, properties {defaultChoiceDisplayText=EXPERT SALE}}
   FortConversationRequirement_HasService_7 {class=FortConversationRequirement_HasService, properties {serviceTag {tagName=Tandem.Service.ExpertSale}}}
  ]
 }
 Back_C_9 {
  id=#28
  class=Back_C
  sub_nodes [ConversationChoiceNode_39 {class=ConversationChoiceNode, properties {defaultChoiceDisplayText=BACK}}]
 }
 ExpertSale_C_21 {
  id=#27
  class=ExpertSale_C
  outputs [#22]
  properties {
   bShowPropertyDetails=false
   loot Tier Group Name=LTG_NPC_Expertise_Pistol_VR
   serviceBriefCollectionKeyOveride=None
   serviceBriefConfigCollection=null
   extraData [
    item {value=/CherrySmokeWeaponGameplay/Gameplay/Pistol/WID_Pistol_Auto_CherrySmoke_Athena_VR}
    cost {value=510}
    remaining {value=1}
    stacksize {value=1}
    Type {value=service.expertsale}
    preview_tags {value=(GameplayTags=((TagName="Conversation.Choice.Category.Service")))}
    BriefWidgetClassPath {value=/FortniteConversationUI/Conversation/WBP_FortBuyBrief_ExpertSale.WBP_FortBuyBrief_ExpertSale_C}
    stockcount {value=1}
   ]
   requirements [{
     cannotUseReasonParameter=""
     failureNodeBehaviour=FailedAndHidden
     participantId {tagName=Conversation.Participant.Player}
     requirement {
      class=UFortControllerRequirement_TokenThreshold_C
      properties {
       itemTemplateId=Token:athena_s36_weaponexpertise_unstablepistol_service_token
       requiredQuantityThresholdConfig {requirement=Equal, quantity {value=2}}
      }
     }
    }]
   serviceCost {curve {rowName=ExpertSale.Cost.EpicPistol.2}}
  }
  sub_nodes [ConversationChoiceNode_32 {class=ConversationChoiceNode, properties {defaultChoiceDisplayText=BUY}}]
 }
 Speech_C_5 {
  id=#15
  class=Speech_C
  outputs [#29, #30]
  properties {
   bUsePerNPC_ContextualMessageLookup=true
   softTaskIcon=&T_UI_NPC_WeaponExpertise
   generalConfig {defaultMessage=You've earned this.}
  }
  sub_nodes [
   FortConversationChoiceWithPreviewNode_5 {class=FortConversationChoiceWithPreviewNode, properties {defaultChoiceDisplayText=EXPERT SALE}}
   FortConversationRequirement_HasService_5 {class=FortConversationRequirement_HasService, properties {serviceTag {tagName=Tandem.Service.ExpertSale}}}
  ]
 }
 ExpertSale_C_13 {
  id=#30
  class=ExpertSale_C
  outputs [#31]
  properties {
   bShowPropertyDetails=false
   loot Tier Group Name=LTG_NPC_Pistol_CherrySmoke_X
   serviceBriefCollectionKeyOveride=None
   serviceBriefConfigCollection=null
   requirements [{
     cannotUseReasonParameter=""
     failureNodeBehaviour=FailedAndHidden
     participantId {tagName=Conversation.Participant.Player}
     requirement {
      class=UFortControllerRequirement_TokenThreshold_C
      properties {
       itemTemplateId=Token:athena_s36_weaponexpertise_unstablepistol_service_token
       requiredQuantityThresholdConfig {requirement=Equal, quantity {value=4}}
      }
     }
    }]
   serviceCost {curve {rowName=ExpertSale.Cost.UnstablePistol.4}}
  }
  sub_nodes [ConversationChoiceNode_20 {class=ConversationChoiceNode, properties {defaultChoiceDisplayText=BUY}}]
 }
 ReturnToStart_C_2 {id=#31, class=ReturnToStart_C}
 Back_C_5 {
  id=#29
  class=Back_C
  sub_nodes [ConversationChoiceNode_26 {class=ConversationChoiceNode, properties {defaultChoiceDisplayText=BACK}}]
 }
 Speech_C_6 {
  id=#14
  class=Speech_C
  outputs [#32, #33]
  properties {
   bUsePerNPC_ContextualMessageLookup=true
   softTaskIcon=&T_UI_NPC_WeaponExpertise
   generalConfig {defaultMessage=You've earned this.}
  }
  sub_nodes [
   FortConversationChoiceWithPreviewNode_2 {class=FortConversationChoiceWithPreviewNode, properties {defaultChoiceDisplayText=EXPERT SALE}}
   FortConversationRequirement_HasService_2 {class=FortConversationRequirement_HasService, properties {serviceTag {tagName=Tandem.Service.ExpertSale}}}
  ]
 }
 ExpertSale_C_14 {
  id=#33
  class=ExpertSale_C
  outputs [#31]
  properties {
   bShowPropertyDetails=false
   loot Tier Group Name=LTG_NPC_Pistol_CherrySmoke_X
   serviceBriefCollectionKeyOveride=None
   serviceBriefConfigCollection=null
   requirements [{
     cannotUseReasonParameter=""
     failureNodeBehaviour=FailedAndHidden
     participantId {tagName=Conversation.Participant.Player}
     requirement {
      class=UFortControllerRequirement_TokenThreshold_C
      properties {
       itemTemplateId=Token:athena_s36_weaponexpertise_unstablepistol_service_token
       requiredQuantityThresholdConfig {requirement=Equal, quantity {value=5}}
      }
     }
    }]
   serviceCost {curve {rowName=ExpertSale.Cost.UnstablePistol.5}}
  }
  sub_nodes [ConversationChoiceNode_15 {class=ConversationChoiceNode, properties {defaultChoiceDisplayText=BUY}}]
 }
 Back_C_4 {
  id=#32
  class=Back_C
  sub_nodes [ConversationChoiceNode_29 {class=ConversationChoiceNode, properties {defaultChoiceDisplayText=BACK}}]
 }
 Speech_C_3 {
  id=#13
  class=Speech_C
  outputs [#34, #35]
  properties {
   bUsePerNPC_ContextualMessageLookup=true
   softTaskIcon=&T_UI_NPC_WeaponExpertise
   generalConfig {defaultMessage=You've earned this.}
  }
  sub_nodes [
   FortConversationChoiceWithPreviewNode_1 {class=FortConversationChoiceWithPreviewNode, properties {defaultChoiceDisplayText=EXPERT SALE}}
   FortConversationRequirement_HasService_1 {class=FortConversationRequirement_HasService, properties {serviceTag {tagName=Tandem.Service.ExpertSale}}}
  ]
 }
 Back_C_2 {
  id=#35
  class=Back_C
  sub_nodes [ConversationChoiceNode_28 {class=ConversationChoiceNode, properties {defaultChoiceDisplayText=BACK}}]
 }
 ExpertSale_C_12 {
  id=#34
  class=ExpertSale_C
  outputs [#31]
  properties {
   bShowPropertyDetails=false
   loot Tier Group Name=LTG_NPC_Pistol_CherrySmoke_X
   serviceBriefCollectionKeyOveride=None
   serviceBriefConfigCollection=null
   requirements [{
     cannotUseReasonParameter=""
     failureNodeBehaviour=FailedAndHidden
     participantId {tagName=Conversation.Participant.Player}
     requirement {
      class=UFortControllerRequirement_TokenThreshold_C
      properties {
       itemTemplateId=Token:athena_s36_weaponexpertise_unstablepistol_service_token
       requiredQuantityThresholdConfig {requirement=Equal, quantity {value=3}}
      }
     }
    }]
   serviceCost {curve {rowName=ExpertSale.Cost.UnstablePistol.3}}
  }
  sub_nodes [ConversationChoiceNode_19 {class=ConversationChoiceNode, properties {defaultChoiceDisplayText=BUY}}]
 }
 Speech_C_1 {
  id=#12
  class=Speech_C
  outputs [#36, #37]
  properties {
   bUsePerNPC_ContextualMessageLookup=true
   softTaskIcon=&T_UI_NPC_WeaponExpertise
   generalConfig {defaultMessage=You've earned this.}
  }
  sub_nodes [
   FortConversationChoiceWithPreviewNode_3 {class=FortConversationChoiceWithPreviewNode, properties {defaultChoiceDisplayText=EXPERT SALE}}
   FortConversationRequirement_HasService_3 {class=FortConversationRequirement_HasService, properties {serviceTag {tagName=Tandem.Service.ExpertSale}}}
  ]
 }
 Back_C_3 {
  id=#37
  class=Back_C
  sub_nodes [ConversationChoiceNode_27 {class=ConversationChoiceNode, properties {defaultChoiceDisplayText=BACK}}]
 }
 ExpertSale_C_11 {
  id=#36
  class=ExpertSale_C
  outputs [#31]
  properties {
   bShowPropertyDetails=false
   loot Tier Group Name=LTG_NPC_Pistol_CherrySmoke_X
   serviceBriefCollectionKeyOveride=None
   serviceBriefConfigCollection=null
   extraData [
    item {value=/CherrySmokeWeaponGameplay/Gameplay/Pistol/Exotic/WID_Pistol_Auto_CherrySmoke_Athena_X}
    cost {value=850}
    remaining {value=1}
    stacksize {value=1}
    Type {value=service.expertsale}
    preview_tags {value=(GameplayTags=((TagName="Conversation.Choice.Category.Service")))}
    BriefWidgetClassPath {value=&WBP_FortBuyBrief_ExpertSale_C}
    stockcount {value=0}
    cannotusereason {value=outofstock}
   ]
   requirements [{
     cannotUseReasonParameter=""
     failureNodeBehaviour=FailedAndHidden
     participantId {tagName=Conversation.Participant.Player}
     requirement {
      class=UFortControllerRequirement_TokenThreshold_C
      properties {
       itemTemplateId=Token:athena_s36_weaponexpertise_unstablepistol_service_token
       requiredQuantityThresholdConfig {requirement=Equal, quantity {value=2}}
      }
     }
    }]
   serviceCost {curve {rowName=ExpertSale.Cost.UnstablePistol.2}}
  }
  sub_nodes [ConversationChoiceNode_18 {class=ConversationChoiceNode, properties {defaultChoiceDisplayText=BUY}}]
 }
 Speech_C_0 {
  id=#11
  class=Speech_C
  outputs [#38, #39]
  properties {
   bUsePerNPC_ContextualMessageLookup=true
   softTaskIcon=&T_UI_NPC_WeaponExpertise
   generalConfig {defaultMessage=You've earned this.}
  }
  sub_nodes [
   FortConversationChoiceWithPreviewNode_6 {class=FortConversationChoiceWithPreviewNode, properties {defaultChoiceDisplayText=EXPERT SALE}}
   FortConversationRequirement_HasService_6 {class=FortConversationRequirement_HasService, properties {serviceTag {tagName=Tandem.Service.ExpertSale}}}
  ]
 }
 Back_C_6 {
  id=#39
  class=Back_C
  sub_nodes [ConversationChoiceNode_38 {class=ConversationChoiceNode, properties {defaultChoiceDisplayText=BACK}}]
 }
 ExpertSale_C_20 {
  id=#38
  class=ExpertSale_C
  outputs [#22]
  properties {
   bShowPropertyDetails=false
   loot Tier Group Name=LTG_NPC_Expertise_Pistol_VR
   serviceBriefCollectionKeyOveride=None
   serviceBriefConfigCollection=null
   extraData [
    item {value=&WID_Pistol_Auto_CherrySmoke_Athena_VR}
    cost {value=600}
    remaining {value=1}
    stacksize {value=1}
    Type {value=service.expertsale}
    preview_tags {value=(GameplayTags=((TagName="Conversation.Choice.Category.Service")))}
    BriefWidgetClassPath {value=&WBP_FortBuyBrief_ExpertSale_C}
    locked {value=1}
    unlock_level {value=-1}
    locked_text {value="LOCTEXT(\"Requires Pistol Expertise\")"}
    stockcount {value=0}
    cannotusereason {value=outofstock}
   ]
   requirements [{
     cannotUseReasonParameter=""
     failureNodeBehaviour=FailedAndHidden
     participantId {tagName=Conversation.Participant.Player}
     requirement {
      class=UFortControllerRequirement_TokenThreshold_C
      properties {
       itemTemplateId=Token:athena_s36_weaponexpertise_unstablepistol_service_token
       requiredQuantityThresholdConfig {requirement=LessThanOrEqual, quantity {value=1}}
      }
     }
    }]
   serviceCost {curve {rowName=ExpertSale.Cost.EpicPistol.1}}
  }
  sub_nodes [ConversationChoiceNode_30 {class=ConversationChoiceNode, properties {defaultChoiceDisplayText=BUY}}]
 }
 Speech_C_7 {
  id=#10
  class=Speech_C
  outputs [#40, #41]
  properties {
   bUsePerNPC_ContextualMessageLookup=true
   softTaskIcon=&T_UI_NPC_WeaponExpertise
   generalConfig {defaultMessage=You've earned this.}
  }
  sub_nodes [
   FortConversationChoiceWithPreviewNode_4 {class=FortConversationChoiceWithPreviewNode, properties {defaultChoiceDisplayText=EXPERT SALE}}
   FortConversationRequirement_HasService_4 {class=FortConversationRequirement_HasService, properties {serviceTag {tagName=Tandem.Service.ExpertSale}}}
  ]
 }
 Back_C_1 {
  id=#41
  class=Back_C
  sub_nodes [ConversationChoiceNode_0 {class=ConversationChoiceNode, properties {defaultChoiceDisplayText=BACK}}]
 }
 ExpertSale_C_10 {
  id=#40
  class=ExpertSale_C
  outputs [#31]
  properties {
   bShowPropertyDetails=false
   loot Tier Group Name=LTG_NPC_Pistol_CherrySmoke_X
   serviceBriefCollectionKeyOveride=None
   serviceBriefConfigCollection=null
   extraData [
    item {value=&WID_Pistol_Auto_CherrySmoke_Athena_X}
    cost {value=1000}
    remaining {value=1}
    stacksize {value=1}
    Type {value=service.expertsale}
    preview_tags {value=(GameplayTags=((TagName="Conversation.Choice.Category.Service")))}
    BriefWidgetClassPath {value=&WBP_FortBuyBrief_ExpertSale_C}
    locked {value=1}
    unlock_level {value=-1}
    locked_text {value="LOCTEXT(\"Requires Pistol Expertise\")"}
    stockcount {value=1}
   ]
   requirements [{
     cannotUseReasonParameter=""
     failureNodeBehaviour=FailedAndHidden
     participantId {tagName=Conversation.Participant.Player}
     requirement {
      class=UFortControllerRequirement_TokenThreshold_C
      properties {
       itemTemplateId=Token:athena_s36_weaponexpertise_unstablepistol_service_token
       requiredQuantityThresholdConfig {requirement=LessThanOrEqual, quantity {value=1}}
      }
     }
    }]
   serviceCost {curve {rowName=ExpertSale.Cost.UnstablePistol.1}}
  }
  sub_nodes [ConversationChoiceNode_17 {class=ConversationChoiceNode, properties {defaultChoiceDisplayText=BUY}}]
 }
]
