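# NPC conversation database for Season 36 Unstable DMR weapon expertise, selling Unstable and Epic DMRs via Expert Sale.
# Three entry points share one tag: sale offers, plus intro and finished dialogues gated on their active quests.
# Each weapon has five duplicated Speech-to-ExpertSale branches requiring the ExpertSale service.
# The player's expertise token count selects the cost curve row, with non-matching purchases hidden through FailedAndHidden requirements.
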ConversationDatabase /BattlePassS36_Quests/Conversations/Expertise/Conversation_Expertise_C
entry_points [{
  tag=Conversation.Entry.Expertise.C
  destinations [#1, #2, #3]
 }]
nodes [
 ConversationEntryPointNode_3 {
  id=#3
  class=ConversationEntryPointNode
  outputs [#4]
  properties {entryTag {tagName=Conversation.Entry.Expertise.C}}
 }
 Speech_C_16 {
  id=#4
  class=Speech_C
  outputs [#5]
  properties {
   softTaskIcon=&T_UI_NPC_MainQuest
   generalConfig {defaultMessage="I'm not easily impressed, but you've done it.", defaultMetadataParameters [IronLilac {value=""}]}
  }
  sub_nodes [
   ConversationChoiceNode_40 {
    class=ConversationChoiceNode
    properties {defaultChoiceDisplayText=Show off your Expertise, choiceTags {gameplayTags [{tagName=Athena.Quests.S36.Conversations.AuraCop.Finished.01}]}}
   }
   FortConversationRequirement_ControllerRequirement_0 {
    class=FortConversationRequirement_ControllerRequirement
    properties {
     nodeName=Require: Is Quest Active
     requirement {
      class=FortControllerRequirement_IsQuestActive
      properties {questDefinition=/BattlePassS36_Quests/Items/QuestItems/WeaponExpertise/UnstableDMR/quest_S36_expertise_UnstableDMR_finished}
     }
    }
   }
  ]
 }
 Speech_C_17 {
  id=#5
  class=Speech_C
  outputs [#6]
  properties {generalConfig {defaultMessage=I can give you a 60% discount on the experimental DMR from my labs.}}
  sub_nodes [ConversationChoiceNode_41 {
    class=ConversationChoiceNode
    properties {defaultChoiceDisplayText=Continue, choiceTags {gameplayTags [{tagName=Athena.Quests.S36.Conversations.AuraCop.Finished.02}]}}
   }]
 }
 ReturnToStart_C_4 {id=#6, class=ReturnToStart_C}
 ConversationEntryPointNode_2 {
  id=#2
  class=ConversationEntryPointNode
  outputs [#7]
  properties {entryTag {tagName=Conversation.Entry.Expertise.C}}
 }
 Speech_C_11 {
  id=#7
  class=Speech_C
  outputs [#8]
  properties {
   softTaskIcon=&T_UI_NPC_MainQuest
   generalConfig {
    defaultMessage="I've bankrolled some... interesting experimental weapons, and we need someone to test them."
    defaultMetadataParameters [IronLilac {value=""}]
   }
  }
  sub_nodes [
   ConversationChoiceNode_13 {
    class=ConversationChoiceNode
    properties {
     defaultChoiceDisplayText=Ask about Unstable Exotic Expertise
     choiceTags {gameplayTags [{tagName=Athena.Quests.S36.Conversations.AuraCop.Intro.01}]}
    }
   }
   FortConversationRequirement_ControllerRequirement_10 {
    class=FortConversationRequirement_ControllerRequirement
    properties {
     nodeName=Require: Is Quest Active
     requirement {
      class=FortControllerRequirement_IsQuestActive
      properties {questDefinition=/BattlePassS36_Quests/Items/QuestItems/WeaponExpertise/UnstableDMR/quest_S36_expertise_UnstableDMR_intro}
     }
    }
   }
  ]
 }
 Speech_C_12 {
  id=#8
  class=Speech_C
  outputs [#9]
  properties {generalConfig {defaultMessage="Interested? I can part with one, for a small donation to Supernova Academy."}}
  sub_nodes [ConversationChoiceNode_14 {
    class=ConversationChoiceNode
    properties {defaultChoiceDisplayText=Continue, choiceTags {gameplayTags [{tagName=Athena.Quests.S36.Conversations.AuraCop.Intro.02}]}}
   }]
 }
 ReturnToStart_C_1 {id=#9, class=ReturnToStart_C}
 ConversationEntryPointNode_1 {
  id=#1
  class=ConversationEntryPointNode
  outputs [
   #10
   #11
   #12
   #13
   #14
   #15
   #16
   #17
   #18
   #19
  ]
  properties {entryTag {tagName=Conversation.Entry.Expertise.C}}
 }
 Speech_C_15 {
  id=#19
  class=Speech_C
  outputs [#20, #21]
  properties {
   bUsePerNPC_ContextualMessageLookup=true
   softTaskIcon=&T_UI_NPC_WeaponExpertise
   generalConfig {defaultMessage=You've earned this.}
  }
  sub_nodes [
   FortConversationChoiceWithPreviewNode_7 {class=FortConversationChoiceWithPreviewNode, properties {defaultChoiceDisplayText=EXPERT SALE}}
   FortConversationRequirement_HasService_7 {class=FortConversationRequirement_HasService, properties {serviceTag {tagName=Tandem.Service.ExpertSale}}}
  ]
 }
 ExpertSale_C_24 {
  id=#21
  class=ExpertSale_C
  outputs [#22]
  properties {
   bShowPropertyDetails=false
   loot Tier Group Name=LTG_NPC_Expertise_DMR_VR
   serviceBriefCollectionKeyOveride=None
   serviceBriefConfigCollection=null
   requirements [{
     cannotUseReasonParameter=""
     failureNodeBehaviour=FailedAndHidden
     participantId {tagName=Conversation.Participant.Player}
     requirement {
      class=UFortControllerRequirement_TokenThreshold_C
      properties {
       itemTemplateId=Token:athena_s36_weaponexpertise_unstabledmr_service_token
       requiredQuantityThresholdConfig {requirement=Equal, quantity {value=5}}
      }
     }
    }]
   serviceCost {curve {rowName=ExpertSale.Cost.EpicDMR.5}}
  }
  sub_nodes [ConversationChoiceNode_33 {class=ConversationChoiceNode, properties {defaultChoiceDisplayText=BUY}}]
 }
 ReturnToStart_C_3 {id=#22, class=ReturnToStart_C}
 Back_C_10 {
  id=#20
  class=Back_C
  sub_nodes [ConversationChoiceNode_32 {class=ConversationChoiceNode, properties {defaultChoiceDisplayText=BACK}}]
 }
 Speech_C_14 {
  id=#18
  class=Speech_C
  outputs [#23, #24]
  properties {
   bUsePerNPC_ContextualMessageLookup=true
   softTaskIcon=&T_UI_NPC_WeaponExpertise
   generalConfig {defaultMessage=You've earned this.}
  }
  sub_nodes [
   FortConversationChoiceWithPreviewNode_11 {class=FortConversationChoiceWithPreviewNode, properties {defaultChoiceDisplayText=EXPERT SALE}}
   FortConversationRequirement_HasService_11 {class=FortConversationRequirement_HasService, properties {serviceTag {tagName=Tandem.Service.ExpertSale}}}
  ]
 }
 ExpertSale_C_23 {
  id=#24
  class=ExpertSale_C
  outputs [#22]
  properties {
   bShowPropertyDetails=false
   loot Tier Group Name=LTG_NPC_Expertise_DMR_VR
   serviceBriefCollectionKeyOveride=None
   serviceBriefConfigCollection=null
   requirements [{
     cannotUseReasonParameter=""
     failureNodeBehaviour=FailedAndHidden
     participantId {tagName=Conversation.Participant.Player}
     requirement {
      class=UFortControllerRequirement_TokenThreshold_C
      properties {
       itemTemplateId=Token:athena_s36_weaponexpertise_unstabledmr_service_token
       requiredQuantityThresholdConfig {requirement=Equal, quantity {value=4}}
      }
     }
    }]
   serviceCost {curve {rowName=ExpertSale.Cost.EpicDMR.4}}
  }
  sub_nodes [ConversationChoiceNode_39 {class=ConversationChoiceNode, properties {defaultChoiceDisplayText=BUY}}]
 }
 Back_C_9 {
  id=#23
  class=Back_C
  sub_nodes [ConversationChoiceNode_38 {class=ConversationChoiceNode, properties {defaultChoiceDisplayText=BACK}}]
 }
 Speech_C_13 {
  id=#17
  class=Speech_C
  outputs [#25, #26]
  properties {
   bUsePerNPC_ContextualMessageLookup=true
   softTaskIcon=&T_UI_NPC_WeaponExpertise
   generalConfig {defaultMessage=You've earned this.}
  }
  sub_nodes [
   FortConversationRequirement_HasService_8 {class=FortConversationRequirement_HasService, properties {serviceTag {tagName=Tandem.Service.ExpertSale}}}
   FortConversationChoiceWithPreviewNode_8 {class=FortConversationChoiceWithPreviewNode, properties {defaultChoiceDisplayText=EXPERT SALE}}
  ]
 }
 ExpertSale_C_22 {
  id=#26
  class=ExpertSale_C
  outputs [#22]
  properties {
   bShowPropertyDetails=false
   loot Tier Group Name=LTG_NPC_Expertise_DMR_VR
   serviceBriefCollectionKeyOveride=None
   serviceBriefConfigCollection=null
   requirements [{
     cannotUseReasonParameter=""
     failureNodeBehaviour=FailedAndHidden
     participantId {tagName=Conversation.Participant.Player}
     requirement {
      class=UFortControllerRequirement_TokenThreshold_C
      properties {
       itemTemplateId=Token:athena_s36_weaponexpertise_unstabledmr_service_token
       requiredQuantityThresholdConfig {requirement=Equal, quantity {value=3}}
      }
     }
    }]
   serviceCost {curve {rowName=ExpertSale.Cost.EpicDMR.3}}
  }
  sub_nodes [ConversationChoiceNode_17 {class=ConversationChoiceNode, properties {defaultChoiceDisplayText=BUY}}]
 }
 Back_C_8 {
  id=#25
  class=Back_C
  sub_nodes [ConversationChoiceNode_15 {class=ConversationChoiceNode, properties {defaultChoiceDisplayText=BACK}}]
 }
 Speech_C_10 {
  id=#16
  class=Speech_C
  outputs [#27, #28]
  properties {
   bUsePerNPC_ContextualMessageLookup=true
   softTaskIcon=&T_UI_NPC_WeaponExpertise
   generalConfig {defaultMessage=You've earned this.}
  }
  sub_nodes [
   FortConversationRequirement_HasService_9 {class=FortConversationRequirement_HasService, properties {serviceTag {tagName=Tandem.Service.ExpertSale}}}
   FortConversationChoiceWithPreviewNode_9 {class=FortConversationChoiceWithPreviewNode, properties {defaultChoiceDisplayText=EXPERT SALE}}
  ]
 }
 ExpertSale_C_21 {
  id=#28
  class=ExpertSale_C
  outputs [#22]
  properties {
   bShowPropertyDetails=false
   loot Tier Group Name=LTG_NPC_Expertise_DMR_VR
   serviceBriefCollectionKeyOveride=None
   serviceBriefConfigCollection=null
   requirements [{
     cannotUseReasonParameter=""
     failureNodeBehaviour=FailedAndHidden
     participantId {tagName=Conversation.Participant.Player}
     requirement {
      class=UFortControllerRequirement_TokenThreshold_C
      properties {
       itemTemplateId=Token:athena_s36_weaponexpertise_unstabledmr_service_token
       requiredQuantityThresholdConfig {requirement=Equal, quantity {value=2}}
      }
     }
    }]
   serviceCost {curve {rowName=ExpertSale.Cost.EpicDMR.2}}
  }
  sub_nodes [ConversationChoiceNode_35 {class=ConversationChoiceNode, properties {defaultChoiceDisplayText=BUY}}]
 }
 Back_C_6 {
  id=#27
  class=Back_C
  sub_nodes [ConversationChoiceNode_34 {class=ConversationChoiceNode, properties {defaultChoiceDisplayText=BACK}}]
 }
 Speech_C_8 {
  id=#15
  class=Speech_C
  outputs [#29, #30]
  properties {
   bUsePerNPC_ContextualMessageLookup=true
   softTaskIcon=&T_UI_NPC_WeaponExpertise
   generalConfig {defaultMessage=You've earned this.}
  }
  sub_nodes [
   FortConversationChoiceWithPreviewNode_10 {class=FortConversationChoiceWithPreviewNode, properties {defaultChoiceDisplayText=EXPERT SALE}}
   FortConversationRequirement_HasService_10 {class=FortConversationRequirement_HasService, properties {serviceTag {tagName=Tandem.Service.ExpertSale}}}
  ]
 }
 ExpertSale_C_20 {
  id=#30
  class=ExpertSale_C
  outputs [#22]
  properties {
   bShowPropertyDetails=false
   loot Tier Group Name=LTG_NPC_Expertise_DMR_VR
   serviceBriefCollectionKeyOveride=None
   serviceBriefConfigCollection=null
   requirements [{
     cannotUseReasonParameter=""
     failureNodeBehaviour=FailedAndHidden
     participantId {tagName=Conversation.Participant.Player}
     requirement {
      class=UFortControllerRequirement_TokenThreshold_C
      properties {
       itemTemplateId=Token:athena_s36_weaponexpertise_unstabledmr_service_token
       requiredQuantityThresholdConfig {requirement=LessThanOrEqual, quantity {value=1}}
      }
     }
    }]
   serviceCost {curve {rowName=ExpertSale.Cost.EpicDMR.1}}
  }
  sub_nodes [ConversationChoiceNode_37 {class=ConversationChoiceNode, properties {defaultChoiceDisplayText=BUY}}]
 }
 Back_C_5 {
  id=#29
  class=Back_C
  sub_nodes [ConversationChoiceNode_36 {class=ConversationChoiceNode, properties {defaultChoiceDisplayText=BACK}}]
 }
 Speech_C_5 {
  id=#14
  class=Speech_C
  outputs [#31, #32]
  properties {
   bUsePerNPC_ContextualMessageLookup=true
   softTaskIcon=&T_UI_NPC_WeaponExpertise
   generalConfig {defaultMessage=You've earned this.}
  }
  sub_nodes [
   FortConversationChoiceWithPreviewNode_5 {class=FortConversationChoiceWithPreviewNode, properties {defaultChoiceDisplayText=EXPERT SALE}}
   FortConversationRequirement_HasService_5 {class=FortConversationRequirement_HasService, properties {serviceTag {tagName=Tandem.Service.ExpertSale}}}
  ]
 }
 ExpertSale_C_18 {
  id=#32
  class=ExpertSale_C
  outputs [#33]
  properties {
   bShowPropertyDetails=false
   loot Tier Group Name=LTG_NPC_CherrySmoke_AR_Unstable_Explosion_R
   serviceBriefCollectionKeyOveride=None
   serviceBriefConfigCollection=null
   requirements [{
     cannotUseReasonParameter=""
     failureNodeBehaviour=FailedAndHidden
     participantId {tagName=Conversation.Participant.Player}
     requirement {
      class=UFortControllerRequirement_TokenThreshold_C
      properties {
       itemTemplateId=Token:athena_s36_weaponexpertise_unstabledmr_service_token
       requiredQuantityThresholdConfig {requirement=Equal, quantity {value=4}}
      }
     }
    }]
   serviceCost {curve {rowName=ExpertSale.Cost.UnstableDMR.4}}
  }
  sub_nodes [ConversationChoiceNode_29 {class=ConversationChoiceNode, properties {defaultChoiceDisplayText=BUY}}]
 }
 ReturnToStart_C_0 {id=#33, class=ReturnToStart_C}
 Back_C_2 {
  id=#31
  class=Back_C
  sub_nodes [ConversationChoiceNode_27 {class=ConversationChoiceNode, properties {defaultChoiceDisplayText=BACK}}]
 }
 Speech_C_6 {
  id=#13
  class=Speech_C
  outputs [#34, #35]
  properties {
   bUsePerNPC_ContextualMessageLookup=true
   softTaskIcon=&T_UI_NPC_WeaponExpertise
   generalConfig {defaultMessage=You've earned this.}
  }
  sub_nodes [
   FortConversationChoiceWithPreviewNode_4 {class=FortConversationChoiceWithPreviewNode, properties {defaultChoiceDisplayText=EXPERT SALE}}
   FortConversationRequirement_HasService_4 {class=FortConversationRequirement_HasService, properties {serviceTag {tagName=Tandem.Service.ExpertSale}}}
  ]
 }
 ExpertSale_C_19 {
  id=#35
  class=ExpertSale_C
  outputs [#33]
  properties {
   bShowPropertyDetails=false
   loot Tier Group Name=LTG_NPC_CherrySmoke_AR_Unstable_Explosion_R
   serviceBriefCollectionKeyOveride=None
   serviceBriefConfigCollection=null
   requirements [{
     cannotUseReasonParameter=""
     failureNodeBehaviour=FailedAndHidden
     participantId {tagName=Conversation.Participant.Player}
     requirement {
      class=UFortControllerRequirement_TokenThreshold_C
      properties {
       itemTemplateId=Token:athena_s36_weaponexpertise_unstabledmr_service_token
       requiredQuantityThresholdConfig {requirement=Equal, quantity {value=5}}
      }
     }
    }]
   serviceCost {curve {rowName=ExpertSale.Cost.UnstableDMR.5}}
  }
  sub_nodes [ConversationChoiceNode_12 {class=ConversationChoiceNode, properties {defaultChoiceDisplayText=BUY}}]
 }
 Back_C_3 {
  id=#34
  class=Back_C
  sub_nodes [ConversationChoiceNode_10 {class=ConversationChoiceNode, properties {defaultChoiceDisplayText=BACK}}]
 }
 Speech_C_3 {
  id=#12
  class=Speech_C
  outputs [#36, #37]
  properties {
   bUsePerNPC_ContextualMessageLookup=true
   softTaskIcon=&T_UI_NPC_WeaponExpertise
   generalConfig {defaultMessage=You've earned this.}
  }
  sub_nodes [
   FortConversationChoiceWithPreviewNode_3 {class=FortConversationChoiceWithPreviewNode, properties {defaultChoiceDisplayText=EXPERT SALE}}
   FortConversationRequirement_HasService_3 {class=FortConversationRequirement_HasService, properties {serviceTag {tagName=Tandem.Service.ExpertSale}}}
  ]
 }
 ExpertSale_C_17 {
  id=#37
  class=ExpertSale_C
  outputs [#33]
  properties {
   bShowPropertyDetails=false
   loot Tier Group Name=LTG_NPC_CherrySmoke_AR_Unstable_Explosion_R
   serviceBriefCollectionKeyOveride=None
   serviceBriefConfigCollection=null
   requirements [{
     cannotUseReasonParameter=""
     failureNodeBehaviour=FailedAndHidden
     participantId {tagName=Conversation.Participant.Player}
     requirement {
      class=UFortControllerRequirement_TokenThreshold_C
      properties {
       itemTemplateId=Token:athena_s36_weaponexpertise_unstabledmr_service_token
       requiredQuantityThresholdConfig {requirement=Equal, quantity {value=3}}
      }
     }
    }]
   serviceCost {curve {rowName=ExpertSale.Cost.UnstableDMR.3}}
  }
  sub_nodes [ConversationChoiceNode_26 {class=ConversationChoiceNode, properties {defaultChoiceDisplayText=BUY}}]
 }
 Back_C_1 {
  id=#36
  class=Back_C
  sub_nodes [ConversationChoiceNode_25 {class=ConversationChoiceNode, properties {defaultChoiceDisplayText=BACK}}]
 }
 Speech_C_1 {
  id=#11
  class=Speech_C
  outputs [#38, #39]
  properties {
   bUsePerNPC_ContextualMessageLookup=true
   softTaskIcon=&T_UI_NPC_WeaponExpertise
   generalConfig {defaultMessage=You've earned this.}
  }
  sub_nodes [
   FortConversationChoiceWithPreviewNode_2 {class=FortConversationChoiceWithPreviewNode, properties {defaultChoiceDisplayText=EXPERT SALE}}
   FortConversationRequirement_HasService_2 {class=FortConversationRequirement_HasService, properties {serviceTag {tagName=Tandem.Service.ExpertSale}}}
  ]
 }
 ExpertSale_C_16 {
  id=#39
  class=ExpertSale_C
  outputs [#33]
  properties {
   bShowPropertyDetails=false
   loot Tier Group Name=LTG_NPC_CherrySmoke_AR_Unstable_Explosion_R
   serviceBriefCollectionKeyOveride=None
   serviceBriefConfigCollection=null
   requirements [{
     cannotUseReasonParameter=""
     failureNodeBehaviour=FailedAndHidden
     participantId {tagName=Conversation.Participant.Player}
     requirement {
      class=UFortControllerRequirement_TokenThreshold_C
      properties {
       itemTemplateId=Token:athena_s36_weaponexpertise_unstabledmr_service_token
       requiredQuantityThresholdConfig {requirement=Equal, quantity {value=2}}
      }
     }
    }]
   serviceCost {curve {rowName=ExpertSale.Cost.UnstableDMR.2}}
  }
  sub_nodes [ConversationChoiceNode_24 {class=ConversationChoiceNode, properties {defaultChoiceDisplayText=BUY}}]
 }
 Back_C_0 {
  id=#38
  class=Back_C
  sub_nodes [ConversationChoiceNode_23 {class=ConversationChoiceNode, properties {defaultChoiceDisplayText=BACK}}]
 }
 Speech_C_7 {
  id=#10
  class=Speech_C
  outputs [#40, #41]
  properties {
   bUsePerNPC_ContextualMessageLookup=true
   softTaskIcon=&T_UI_NPC_WeaponExpertise
   generalConfig {defaultMessage=You've earned this.}
  }
  sub_nodes [
   FortConversationChoiceWithPreviewNode_6 {class=FortConversationChoiceWithPreviewNode, properties {defaultChoiceDisplayText=EXPERT SALE}}
   FortConversationRequirement_HasService_6 {class=FortConversationRequirement_HasService, properties {serviceTag {tagName=Tandem.Service.ExpertSale}}}
  ]
 }
 ExpertSale_C_15 {
  id=#41
  class=ExpertSale_C
  outputs [#33]
  properties {
   bShowPropertyDetails=false
   loot Tier Group Name=LTG_NPC_CherrySmoke_AR_Unstable_Explosion_R
   serviceBriefCollectionKeyOveride=None
   serviceBriefConfigCollection=null
   requirements [{
     cannotUseReasonParameter=""
     failureNodeBehaviour=FailedAndHidden
     participantId {tagName=Conversation.Participant.Player}
     requirement {
      class=UFortControllerRequirement_TokenThreshold_C
      properties {
       itemTemplateId=Token:athena_s36_weaponexpertise_unstabledmr_service_token
       requiredQuantityThresholdConfig {requirement=LessThanOrEqual, quantity {value=1}}
      }
     }
    }]
   serviceCost {curve {rowName=ExpertSale.Cost.UnstableDMR.1}}
  }
  sub_nodes [ConversationChoiceNode_31 {class=ConversationChoiceNode, properties {defaultChoiceDisplayText=BUY}}]
 }
 Back_C_7 {
  id=#40
  class=Back_C
  sub_nodes [ConversationChoiceNode_30 {class=ConversationChoiceNode, properties {defaultChoiceDisplayText=BACK}}]
 }
]
